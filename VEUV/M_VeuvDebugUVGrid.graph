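# M_VeuvDebugUVGrid is a UI-domain material that defines a debug UV grid over a muted red-green UV gradient.
# A single Custom HLSL node, fed by TextureCoordinate, estimates zoom from fwidth and picks power-of-ten grid spacing.
# It blends faint, anti-aliased gray sub-grid lines over the gradient.
# Major grid lines are computed but zeroed out, with a TODO noting they clash with background lines.
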Material /VEUV/M_VeuvDebugUVGrid
graph {
 properties {bAutomaticallySetUsageInEditor=false, materialDomain=MD_UI, usedShadingModels=Default Lit}
 nodes [
  MaterialExpressionTextureCoordinate_0 = TextureCoordinate()
  Custom!(UV=@MaterialExpressionTextureCoordinate_0) {
   code="float2 UV_In = UV;\r\n\r\n// Pixel size in UV space (implicit zoom level)\r\nfloat PixelSize = max(fwidth(UV_In.x), fwidth(UV_In.y));\r\n\r\n// Auto-select grid spacing from zoom\r\nfloat LogStep = log10(PixelSize * 100.0);\r\nfloat GridStep = pow(10.0, ceil(LogStep));\r\nfloat SubStep = GridStep * 0.1;\r\n\r\n// Major grid lines\r\nfloat2 G = frac(UV_In / GridStep);\r\nfloat2 DG = min(G, 1.0 - G) / (PixelSize / GridStep);\r\nfloat Grid = 1.0 - smoothstep(0.0, 1.5, min(DG.x, DG.y));\r\n\r\n// TODO: Removed, clashes with background lines\r\nGrid = 0;\r\n\r\n// Sub-grid lines\r\nfloat2 SG = frac(UV_In / SubStep);\r\nfloat2 DSG = min(SG, 1.0 - SG) / (PixelSize / SubStep);\r\nfloat SubGrid = (1.0 - smoothstep(0.0, 1.5, min(DSG.x, DSG.y))) * 0.2;\r\n\r\n// Muted UV-colored base\r\nfloat3 Base = float3(frac(UV_In.x), frac(UV_In.y), 0.0) * 0.5;\r\n\r\n// Grid overlay\r\nfloat3 GridColor = float3(0.6, 0.6, 0.6);\r\nfloat3 Result = lerp(Base, GridColor, max(Grid, SubGrid));\r\n\r\nreturn float4(Result, 1.0);"
   showCode=true
   inputs [{inputName=UV, input {expression=&self:MaterialExpressionTextureCoordinate_0}}]
  }
 ]
}
